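# Physics asset for the F_MED_RustyBolt backpack cosmetic, supplying its simplified collision shape.
# It holds a single body, F_MED_RustyBolt_MESH_LOD0, approximated by one rotated, slightly offset capsule that is nearly as long as its radius.
# It defines no constraints, so there is no articulation between parts.
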PhysicsAsset /BRCosmetics/Accessories/FORT_Backpacks/F_MED_RustyBolt/Meshes/F_MED_RustyBolt_Backpack_PhysicsAsset
bodies {F_MED_RustyBolt_MESH_LOD0 [{shapeType=Capsule, length=9.9845, radius=9.9496, center=<0.1042 -1.8588 2.7932>, rotation=<rot 0 -89.9836 -58.3573>}]}
constraints []
